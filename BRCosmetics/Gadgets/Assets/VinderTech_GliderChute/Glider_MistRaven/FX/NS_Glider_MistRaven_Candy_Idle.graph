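# Niagara idle effect for the Candy MistRaven glider, combining airglow, swaying smoke meshes, flickering flares, GPU fire, sparks and feathers shed from a skeletal mesh.
# User.Variant indexes arrays for colors and spawn counts, with base smoke and flare emitters disabled.
# FrontEnd modifiers choose between gameplay and front-end copies of the variant smoke and flares.
# The system timeline scales feather and fire spawn rates, fire acceleration and smoke material parameters.
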NiagaraSystem /BRCosmetics/Gadgets/Assets/VinderTech_GliderChute/Glider_MistRaven/FX/NS_Glider_MistRaven_Candy_Idle
dependencies {
 usedDataInterfaces [
  &NiagaraDataInterfaceCurve
  &NiagaraDataInterfaceArrayInt32
  &NiagaraDataInterfaceArrayFloat3
  &NiagaraDataInterfaceArrayColor
  &NiagaraDataInterfaceColorCurve
  &NiagaraDataInterfaceSpriteRendererInfo
  &NiagaraDataInterfaceArrayFloat
  &NiagaraDataInterfaceSkeletalMesh
 ]
 usedDynamicInputs [&SetBoolByFloatComparison]
 usedModules [
  &SystemState
  &Timeline
  &self:SystemSpawnScript.SystemScriptSource.SystemScriptGraph.NiagaraNodeAssignment_0.SetVariables_#1
  &self:SystemSpawnScript.SystemScriptSource.SystemScriptGraph.NiagaraNodeAssignment_1.SetVariables_#2
 ]
}
emitters {
 Airglow {
  emitterUpdateScript [
   &EmitterState(Life Cycle Mode=::NewEnumerator1, Loop Behavior=::NewEnumerator1, Loop Duration Mode=::NewEnumerator1, UseLoopDelay=false)
   &SpawnBurst_Instantaneous(Use Spawn Probability=false, UseLoopCountLimit=false)
  ]
  particleSpawnScript [
   &InitializeParticle(
    Color=&SelectLinearColorFromArray(
     Array Sampling Mode=::NewEnumerator1
     Direct Array Index=&MakeIntFromFloat(Float=$User.Variant)
     Color Selection Array=DataInterface() {propertyValues {ColorData [<rgba 0.4792 0.0078 0.0565 1>, <rgba 0.4792 0.0001 0 0.4>]}}
     Random Seed=?
    )
    Lifetime=5
    Position Offset=$System.FlareOffset
    Sprite Size Mode=::NewEnumerator3
    Uniform Sprite Size=20
    UsePositionOffset=true
    Write Material Random=true
   )
   &CameraOffset()
  ]
  particleUpdateScript [
   &ParticleState(Kill Particles When Lifetime Has Elapsed=false, Let Infinitely Lived Particles Die When Emitter Deactivates=true)
   &DynamicMaterialParameters(
    Index 0 Param 1=3
    Index 0 Param 2=5
    Module.Index0Float0Write=true
    Module.Index0Float1Write=true
    Module.Index0Float2Write=true
    Module.Index0Float3Write=true
   )
   &ScaleColor(
    Scale Alpha=&FloatFromCurve(
     FloatCurve=DataInterface() {propertyValues {Curve {defaultValue=1, keys [{tangentMode=RCTM_User, time=0, value=0}, {tangentMode=RCTM_User, time=0.5, value=1}]}}}
     CurveIndex=$Particles.Age
     Scale Curve=2
    )
    ScaleA=true
    ScaleRGB=true
   )
   &SpriteFacingAndAlignment(Alignment Coordinate Space=::World, Sprite Alignment=<0 0 1>, Sprite Facing=<0 -1 0>, WriteAlignment=true, WriteFacing=true)
   /SetVariables(
    Particles.Facing2:Vector3f=<1 0 0>
    Particles.SpriteSize2=&Multiply_Vector2D(A=$Particles.SpriteSize, B=<0.5 1>)
   ) {isSetParameters=true}
   /SetVariables(Particles.Facing3:Vector3f=<0.001 0 1>, Particles.alignment3:Vector3f=<0 0 1>) {isSetParameters=true}
  ]
  rendererClasses [&NiagaraSpriteRendererProperties]
  renderers [{rendererClass=&NiagaraSpriteRendererProperties}]
 }
 BottomSmoke {
  bEnabled=false
  emitterUpdateScript [
   &EmitterState(Life Cycle Mode=::NewEnumerator1, Loop Behavior=::NewEnumerator1, Loop Duration Mode=::NewEnumerator1, UseLoopDelay=false)
   &SpawnBurst_Instantaneous(
    Spawn Count=&SelectIntFromArray(
     Direct Array Index=&MakeIntFromFloat(Float=$User.Variant)
     Int Selection Array=DataInterface() {propertyValues {IntData [1,0]}}
    )
    Use Spawn Probability=false
    UseLoopCountLimit=false
   )
  ]
  particleSpawnScript [
   &InitializeParticle(
    Color=<rgba 0.625 0.2604 0.3288 1>
    Lifetime=5
    Mesh Scale=&Multiply_VectorByFloat(Vector=<0.1 0.15 1>, Float=4)
    Mesh Scale Mode=::NewEnumerator4
    Position Offset=&Add_Vector(A=$System.FlareOffset, B=<-10 2 0>)
    UsePositionOffset=true
    Write Material Random=true
   )
   &InitialMeshOrientation(Mesh Orientation Mode=::NewEnumerator2, Use Rotation Vector=false)
   &CameraOffset(Camera Offset Amount=4)
   /SetVariables(Particles.Color1:LinearColor=&Multiply_LinearColorByFloat(Linear Color=<rgba 0.5781 0.1156 0.2023 1>, Float=1)) {isSetParameters=true, module=SetVariables_#3}
   /SetVariables(Particles.Color2=&Multiply_LinearColorByFloat(Linear Color=<rgba 1 0.4375 0.543 1>, Float=1)) {isSetParameters=true, module=SetVariables_#4}
   /SetVariables(Particles.Color3:LinearColor=&Multiply_LinearColorByFloat(Linear Color=<rgba 0.9479 0.1333 0.286 1>, Float=0.3)) {isSetParameters=true, module=SetVariables_#5}
  ]
  particleUpdateScript [
   &ParticleState(Kill Particles When Lifetime Has Elapsed=false)
   &ScaleMeshSize(Scale Factor=&MakeVector(X=1, Y=1, Z=1))
   /SetVariables(Particles.MeshOrientation=&MakeQuatFromAxisAngle(
     Axis=<0 1 0>
     AngleInDegrees=&Lerp_Float(
      A=-100
      B=-95
      Alpha=&RemapRange(
       Input Value To Remap=&Waveform(
        Waveform Count=::NewEnumerator1
        Global Phase=&Add_Float(A=$Emitter.Age, B=1)
        "Waveform [1]"=::NewEnumerator0
        "[1] Frequency"=0.05
        "[1] Amplitude Min/Max"=<-1 1>
        "Waveform [2] Blend"=::NewEnumerator2
        "Waveform [2]"=::NewEnumerator1
        "[2] Input Position"=0
        "[2] Phase"=$Emitter.Age
        "[2] Phase Scale"=1
        "[2] Compound Frequency"=2
        "[2] Frequency"=0.1
        "[2] Amplitude Min/Max"=<0.5 1>
        "[2] Scale"=1
        "[2] Bias"=0
        "[2] Exponent"=1
        "[2] Pulse Duration"=0.1
        "[2] Clip Triangle Min/Max"=<0 0>
        "[2] Random Seed"=0
       )
       Input Min=-1
       Input Max=1
       Output Min=0
       Output Max=1
      )
     )
    )) {isSetParameters=true, module=SetVariables_#6}
   &InterpolateOverTime(
    Rate Of Change=6
    Target Vector Value:Vector3f=&ConvertPositionToVector(
     Input Position=&TransformPosition(Position=$Particles.Position, Source Space=::Local, Destination Space=::World)
     Convert To Absolute Large World Vector=false
    )
    Variable Type=::NewEnumerator2
   )
   /SetVariables(Particles.Interpolate Dynamic=&MakeFloat4FromVectorAndFloat("Vector (RGB)"=$StackContext.InterpolateOverTime.InterpolateOverTime_V3_.Moving Average, "Float (Alpha)"=0.3)) {isSetParameters=true, module=SetVariables_#7}
   &ScaleColor(ScaleA=true, ScaleRGB=false)
   &DynamicMaterialParameters(
    Index 0 Param 2=&FloatFromCurve(
     FloatCurve=DataInterface() {propertyValues {Curve {defaultValue=1, keys [{tangentMode=RCTM_User, time=1.5, value=0}, {tangentMode=RCTM_User, time=2.5, value=1}]}}}
     CurveIndex=$Particles.Age
    )
    Index 0 Param 4=&Frac_Float(Float=$System.FN_Integrator002.Value)
    Index 2 Param 2=&Lerp_Float(A=0, B=20, Alpha=$System.Timeline.NormalizedTime)
    Module.Index0Param0Write=false
    Module.Index0Param1Write=true
    Module.Index0Param2Write=false
    Module.Index0Param3Write=true
    Module.Index2Param0Write=false
    Module.Index2Param1Write=true
    Module.Index2Param2Write=false
    Module.Index2Param3Write=false
    Write Parameter Index 2=true
   )
  ]
  rendererClasses [&NiagaraMeshRendererProperties]
  renderers [
   {rendererClass=&NiagaraMeshRendererProperties, rendererIndex=0}
   {rendererClass=&NiagaraMeshRendererProperties, rendererIndex=1}
   {rendererClass=&NiagaraMeshRendererProperties, rendererIndex=2}
   {rendererClass=&NiagaraMeshRendererProperties, rendererIndex=3}
  ]
 }
 BottomSmokeVariant {
  emitterUpdateScript [
   &EmitterState(Life Cycle Mode=::NewEnumerator1, Loop Behavior=::NewEnumerator1, Loop Duration Mode=::NewEnumerator1, UseLoopDelay=false)
   &SpawnBurst_Instantaneous(
    Spawn Count=&MakeIntFromFloat(Float=&FN_FrontEndFloatModifier(
      Gameplay=&MakeFloatFromInt(Int=&SelectIntFromArray(
        Direct Array Index=&MakeIntFromFloat(Float=$User.Variant)
        Int Selection Array=DataInterface() {propertyValues {IntData [0,1]}}
       ))
      Front End=0
     ))
    Use Spawn Probability=false
    UseLoopCountLimit=false
   )
  ]
  particleSpawnScript [
   &InitializeParticle(
    Color=<rgba 0.625 0.2604 0.3288 1>
    Lifetime=5
    Mesh Scale=&Multiply_VectorByFloat(Vector=<0.1 0.15 1>, Float=4)
    Mesh Scale Mode=::NewEnumerator4
    Position Offset=&Add_Vector(A=$System.FlareOffset, B=<-10 2 0>)
    UsePositionOffset=true
    Write Material Random=true
   )
   &InitialMeshOrientation(Mesh Orientation Mode=::NewEnumerator2, Use Rotation Vector=false)
   &CameraOffset(Camera Offset Amount=4)
   /SetVariables(Particles.Color1:LinearColor=&Multiply_LinearColorByFloat(
     Linear Color=&MultiplyLinearColor(
      Linear Color=<rgba 0.5781 0.1156 0.2023 1>
      Scale Factor=&MakeVector4FromVectorAndFloat(
       "Vector (RGB)"=&FN_FrontEndVector3Modifier(Gameplay=<1 1 1>, Front End=<0.6 1 5>)
       "Float (Alpha)"=2
      )
     )
     Float=1
    )) {isSetParameters=true, module=SetVariables_#3}
   /SetVariables(Particles.Color2=&Multiply_LinearColorByFloat(
     Linear Color=&MultiplyLinearColor(
      Linear Color=<rgba 1 0.4375 0.543 1>
      Scale Factor=&MakeVector4FromVectorAndFloat(
       "Vector (RGB)"=&FN_FrontEndVector3Modifier(Gameplay=<1 1 1>, Front End=<0.6 1 5>)
       "Float (Alpha)"=1
      )
     )
     Float=1
    )) {isSetParameters=true, module=SetVariables_#4}
   /SetVariables(Particles.Color3:LinearColor=&Multiply_LinearColorByFloat(Linear Color=<rgba 0.0124 0.0116 0.0156 5>, Float=0.45)) {isSetParameters=true, module=SetVariables_#5}
  ]
  particleUpdateScript [
   &ParticleState(Kill Particles When Lifetime Has Elapsed=false)
   &ScaleMeshSize(Scale Factor=&MakeVector(X=1, Y=1, Z=1))
   /SetVariables(Particles.MeshOrientation=&MakeQuatFromAxisAngle(
     Axis=<0 1 0>
     AngleInDegrees=&Lerp_Float(
      A=-100
      B=-95
      Alpha=&RemapRange(
       Input Value To Remap=&Waveform(
        Waveform Count=::NewEnumerator1
        Global Phase=&Add_Float(A=$Emitter.Age, B=1)
        "Waveform [1]"=::NewEnumerator0
        "[1] Frequency"=0.05
        "[1] Amplitude Min/Max"=<-1 1>
        "Waveform [2] Blend"=::NewEnumerator2
        "Waveform [2]"=::NewEnumerator1
        "[2] Input Position"=0
        "[2] Phase"=$Emitter.Age
        "[2] Phase Scale"=1
        "[2] Compound Frequency"=2
        "[2] Frequency"=0.1
        "[2] Amplitude Min/Max"=<0.5 1>
        "[2] Scale"=1
        "[2] Bias"=0
        "[2] Exponent"=1
        "[2] Pulse Duration"=0.1
        "[2] Clip Triangle Min/Max"=<0 0>
        "[2] Random Seed"=0
       )
       Input Min=-1
       Input Max=1
       Output Min=0
       Output Max=1
      )
     )
    )) {isSetParameters=true, module=SetVariables_#6}
   &InterpolateOverTime(
    Rate Of Change=6
    Target Vector Value:Vector3f=&ConvertPositionToVector(
     Input Position=&TransformPosition(Position=$Particles.Position, Source Space=::Local, Destination Space=::World)
     Convert To Absolute Large World Vector=false
    )
    Variable Type=::NewEnumerator2
   )
   /SetVariables(Particles.Interpolate Dynamic=&MakeFloat4FromVectorAndFloat("Vector (RGB)"=$StackContext.InterpolateOverTime.InterpolateOverTime_V3_.Moving Average, "Float (Alpha)"=0.3)) {isSetParameters=true, module=SetVariables_#7}
   &ScaleColor(ScaleA=true, ScaleRGB=false)
   &DynamicMaterialParameters(
    Index 0 Param 2=&FloatFromCurve(
     FloatCurve=DataInterface() {propertyValues {Curve {defaultValue=1, keys [{tangentMode=RCTM_User, time=1.5, value=0}, {tangentMode=RCTM_User, time=2.5, value=1}]}}}
     CurveIndex=$Particles.Age
    )
    Index 0 Param 4=&Frac_Float(Float=$System.FN_Integrator002.Value)
    Index 2 Param 2=&Lerp_Float(A=0, B=20, Alpha=$System.Timeline.NormalizedTime)
    Module.Index0Param0Write=false
    Module.Index0Param1Write=true
    Module.Index0Param2Write=false
    Module.Index0Param3Write=true
    Module.Index2Param0Write=false
    Module.Index2Param1Write=true
    Module.Index2Param2Write=false
    Module.Index2Param3Write=false
    Write Parameter Index 2=true
   )
  ]
  rendererClasses [&NiagaraMeshRendererProperties]
  renderers [
   {rendererClass=&NiagaraMeshRendererProperties, rendererIndex=0}
   {rendererClass=&NiagaraMeshRendererProperties, rendererIndex=1}
   {rendererClass=&NiagaraMeshRendererProperties, rendererIndex=2}
   {rendererClass=&NiagaraMeshRendererProperties, rendererIndex=3}
  ]
 }
 BottomSmokeVariant_FE {
  emitterUpdateScript [
   &EmitterState(Life Cycle Mode=::NewEnumerator1, Loop Behavior=::NewEnumerator1, Loop Duration Mode=::NewEnumerator1, UseLoopDelay=false)
   &SpawnBurst_Instantaneous(
    Spawn Count=&MakeIntFromFloat(Float=&FN_FrontEndFloatModifier(
      Gameplay=0
      Front End=&MakeFloatFromInt(Int=&SelectIntFromArray(
        Direct Array Index=&MakeIntFromFloat(Float=$User.Variant)
        Int Selection Array=DataInterface() {propertyValues {IntData [0,1]}}
       ))
     ))
    Use Spawn Probability=false
    UseLoopCountLimit=false
   )
  ]
  particleSpawnScript [
   &InitializeParticle(
    Color=<rgba 0.625 0.2604 0.3288 1>
    Lifetime=5
    Mesh Scale=&Multiply_VectorByFloat(Vector=<0.1 0.15 1>, Float=4)
    Mesh Scale Mode=::NewEnumerator4
    Position Offset=&Add_Vector(A=$System.FlareOffset, B=<-10 2 0>)
    UsePositionOffset=true
    Write Material Random=true
   )
   &InitialMeshOrientation(Mesh Orientation Mode=::NewEnumerator2, Use Rotation Vector=false)
   &CameraOffset(Camera Offset Amount=4)
   /SetVariables(Particles.Color1:LinearColor=&Multiply_LinearColorByFloat(
     Linear Color=&MultiplyLinearColor(
      Linear Color=<rgba 0.5781 0.1156 0.2023 1>
      Scale Factor=&MakeVector4FromVectorAndFloat(
       "Vector (RGB)"=&FN_FrontEndVector3Modifier(Gameplay=<1 1 1>, Front End=<1 1 1>)
       "Float (Alpha)"=1
      )
     )
     Float=1
    )) {isSetParameters=true, module=SetVariables_#3}
   /SetVariables(Particles.Color2=&Multiply_LinearColorByFloat(
     Linear Color=&MultiplyLinearColor(
      Linear Color=<rgba 1 0.4375 0.543 1>
      Scale Factor=&MakeVector4FromVectorAndFloat(
       "Vector (RGB)"=&FN_FrontEndVector3Modifier(Gameplay=<1 1 1>, Front End=<1 1 1>)
       "Float (Alpha)"=1
      )
     )
     Float=1
    )) {isSetParameters=true, module=SetVariables_#4}
   /SetVariables(Particles.Color3:LinearColor=&Multiply_LinearColorByFloat(Linear Color=<rgba 0.0124 0.0116 0.0156 5>, Float=0.45)) {isSetParameters=true, module=SetVariables_#5}
  ]
  particleUpdateScript [
   &ParticleState(Kill Particles When Lifetime Has Elapsed=false)
   &ScaleMeshSize(Scale Factor=&MakeVector(X=1, Y=1, Z=1))
   /SetVariables(Particles.MeshOrientation=&MakeQuatFromAxisAngle(
     Axis=<0 1 0>
     AngleInDegrees=&Lerp_Float(
      A=-100
      B=-95
      Alpha=&RemapRange(
       Input Value To Remap=&Waveform(
        Waveform Count=::NewEnumerator1
        Global Phase=&Add_Float(A=$Emitter.Age, B=1)
        "Waveform [1]"=::NewEnumerator0
        "[1] Frequency"=0.05
        "[1] Amplitude Min/Max"=<-1 1>
        "Waveform [2] Blend"=::NewEnumerator2
        "Waveform [2]"=::NewEnumerator1
        "[2] Input Position"=0
        "[2] Phase"=$Emitter.Age
        "[2] Phase Scale"=1
        "[2] Compound Frequency"=2
        "[2] Frequency"=0.1
        "[2] Amplitude Min/Max"=<0.5 1>
        "[2] Scale"=1
        "[2] Bias"=0
        "[2] Exponent"=1
        "[2] Pulse Duration"=0.1
        "[2] Clip Triangle Min/Max"=<0 0>
        "[2] Random Seed"=0
       )
       Input Min=-1
       Input Max=1
       Output Min=0
       Output Max=1
      )
     )
    )) {isSetParameters=true, module=SetVariables_#6}
   &InterpolateOverTime(
    Rate Of Change=6
    Target Vector Value:Vector3f=&ConvertPositionToVector(
     Input Position=&TransformPosition(Position=$Particles.Position, Source Space=::Local, Destination Space=::World)
     Convert To Absolute Large World Vector=false
    )
    Variable Type=::NewEnumerator2
   )
   /SetVariables(Particles.Interpolate Dynamic=&MakeFloat4FromVectorAndFloat("Vector (RGB)"=$StackContext.InterpolateOverTime.InterpolateOverTime_V3_.Moving Average, "Float (Alpha)"=0.3)) {isSetParameters=true, module=SetVariables_#7}
   &ScaleColor(ScaleA=true, ScaleRGB=false)
   &DynamicMaterialParameters(
    Index 0 Param 2=&FloatFromCurve(
     FloatCurve=DataInterface() {propertyValues {Curve {defaultValue=1, keys [{tangentMode=RCTM_User, time=1.5, value=0}, {tangentMode=RCTM_User, time=2.5, value=1}]}}}
     CurveIndex=$Particles.Age
    )
    Index 0 Param 4=&Frac_Float(Float=$System.FN_Integrator002.Value)
    Index 2 Param 2=&Lerp_Float(A=0, B=20, Alpha=$System.Timeline.NormalizedTime)
    Module.Index0Param0Write=false
    Module.Index0Param1Write=true
    Module.Index0Param2Write=false
    Module.Index0Param3Write=true
    Module.Index2Param0Write=false
    Module.Index2Param1Write=true
    Module.Index2Param2Write=false
    Module.Index2Param3Write=false
    Write Parameter Index 2=true
   )
  ]
  rendererClasses [&NiagaraMeshRendererProperties]
  renderers [
   {rendererClass=&NiagaraMeshRendererProperties, rendererIndex=0}
   {rendererClass=&NiagaraMeshRendererProperties, rendererIndex=1}
   {rendererClass=&NiagaraMeshRendererProperties, rendererIndex=2}
   {rendererClass=&NiagaraMeshRendererProperties, rendererIndex=3}
  ]
 }
 Feathers {
  emitterUpdateScript [
   &EmitterState(Life Cycle Mode=::NewEnumerator1, Loop Behavior=::NewEnumerator1, Loop Delay=2, Loop Duration Mode=::NewEnumerator1, UseLoopDelay=true)
   &SpawnRate(SpawnRate=&Multiply_Float(A=5, B=$System.Timeline.NormalizedTime), Use Spawn Probability=false)
  ]
  particleSpawnScript [
   &InitializeParticle(
    Color=<rgba 1 0.0162 0.1179 1>
    Lifetime=&RandomRangeFloat(Minimum=0.5, Maximum=1.5)
    Mesh Scale Mode=::NewEnumerator1
    Mesh Uniform Scale Max=0.2
    Mesh Uniform Scale Min=0.1
    Sprite Size Mode=::NewEnumerator3
    UsePositionOffset=false
    Write Material Random=true
   )
   &SkeletalMeshLocation(
    MaskSpawnGroup=false
    Mesh Sampling Type=::NewEnumerator5
    OverrideRandomSeed=false
    Sampled Position Offset=&RandomRangeVector(Minimum=<-10 -10 -10>, Maximum=<10 10 10>)
    Skeletal Mesh=$System.Skeletal Mesh
   )
   &AddVelocity()
   &AddVelocity(Velocity=<0 -1 0>, Velocity Speed Scale=200) {module=AddVelocity002}
   &InitialMeshOrientation(Use Rotation Vector=false)
   &AddVelocity(Velocity=<0 0 -50>) {module=AddVelocity001}
   /SetVariables(Particles.VisibilityTag=&MakeIntFromFloat(Float=$User.Variant)) {isSetParameters=true}
  ]
  particleUpdateScript [
   &ParticleState()
   &ScaleMeshSize(Scale Factor=&VectorFromFloat(Value=&FloatFromCurve(
      FloatCurve=DataInterface() {propertyValues {Curve {
         defaultValue=1
         keys [
          {arriveTangent=-0.0047, interpMode=RCIM_Cubic, leaveTangent=-0.0047, tangentMode=RCTM_User, time=0, value=0}
          {arriveTangentWeight=0.0818, interpMode=RCIM_Cubic, leaveTangentWeight=0.3251, tangentMode=RCTM_User, time=0.2, value=1}
          {arriveTangentWeight=0.2835, interpMode=RCIM_Cubic, leaveTangentWeight=0.1521, tangentMode=RCTM_User, time=0.7, value=1}
          {arriveTangent=-0.0405, interpMode=RCIM_Cubic, leaveTangent=-0.0405, tangentMode=RCTM_User, time=1, value=0}
         ]
        }}}
      Scale Curve=1.25
     )))
   &InheritSourceMovement(ApplyForce=false, ApplyPosition=true, ApplyPositionFalloff=false)
   &CurlNoiseForce(Bias Noise=false, Noise Frequency=25, Noise Strength=100, Pan Noise=false)
   &GravityForce(Gravity=<0 0 -200>)
   &MeshRotationForce(Rotation=&VectorFromFloat(Value=&Multiply_Float(
      A=&UniformAOrBFloat(
       A=&RandomRangeFloat(Minimum=-0.5, Maximum=-1)
       B=&RandomRangeFloat(Minimum=0.5, Maximum=1)
       Evaluation Type=::NewEnumerator0
       Recalculate Random Each Loop=true
       Override Seed=false
      )
      B=3
     )))
   &SolveForcesAndVelocity(Clamp Velocity=false, Limit Acceleration=false)
  ]
  rendererClasses [&NiagaraMeshRendererProperties]
  renderers [
   {rendererClass=&NiagaraMeshRendererProperties, rendererIndex=0}
   {rendererClass=&NiagaraMeshRendererProperties, rendererIndex=1}
  ]
 }
 Fire001 {
  simTarget=GPUComputeSim
  emitterUpdateScript [
   &EmitterState(
    Inactive Response=::NewEnumerator1
    Life Cycle Mode=::NewEnumerator1
    Loop Behavior=::NewEnumerator1
    Loop Delay=2
    Loop Duration Mode=::NewEnumerator1
    UseLoopDelay=true
   )
   &SpawnRate(
    Spawn Probability=0.8
    SpawnRate=&Lerp_Float(A=5, B=10, Alpha=$System.Timeline.NormalizedTime)
    Use Spawn Probability=true
   )
  ]
  particleSpawnScript [
   /SetVariables(Particles.Random size:NiagaraFloat=&RandomRangeFloat(Minimum=0.2, Maximum=0.25)) {isSetParameters=true}
   &InitializeParticle(
    Color=&SelectLinearColorFromArray(
     Array Sampling Mode=::NewEnumerator1
     Direct Array Index=&MakeIntFromFloat(Float=$User.Variant)
     Color Selection Array=DataInterface() {propertyValues {ColorData [<rgba 5 0 0 1>, <rgba 0.01 0.01 0.01 1>]}}
     Random Seed=?
    )
    Lifetime Max=1.2
    Lifetime Min=0.8
    Lifetime Mode=::NewEnumerator1
    Position Offset=$System.FlareOffset
    Sprite Rotation Mode=::NewEnumerator2
    Sprite Size Mode=::NewEnumerator3
    Sprite UV Mode=::NewEnumerator3
    Uniform Sprite Size=&Multiply_Float(A=0, B=$Particles.Random size)
    UsePositionOffset=true
    Write Material Random=true
   ) {module=InitializeParticle001}
   &DynamicMaterialParameters(
    Index 0 Param 2=&RandomRangeFloat(Minimum=2, Maximum=4)
    Index 0 Param 4=2.5
    Module.Param0WriteEnabled=true
    Module.Param1WriteEnabled=true
    Module.Param2WriteEnabled=false
    Module.Param3WriteEnabled=true
   ) {module=DynamicMaterialParameters001}
   &AddVelocity(
    Rotation Coordinate Space=::Simulation
    Velocity=&FN_FrontEndVector3Modifier(Gameplay=<-1 0 0>, Front End=<-1 -0.25 0>)
    Velocity Speed Scale=150
   ) {module=AddVelocity001}
   &CameraOffset() {module=CameraOffset001}
   /SetVariables() {isSetParameters=true}
   /SetVariables(Particles.DY3:Vector4f=&MakeVector4(X=3, Y=1, Z=1, W=1)) {isSetParameters=true}
   &ShapeLocation(MaskSpawnGroup=false, Override Seed=false, Sphere Radius=5, UseOwnerScale=true)
  ]
  particleUpdateScript [
   /SetVariables(Particles.Size Scale:NiagaraFloat=&FloatFromCurve(
     FloatCurve=DataInterface() {propertyValues {Curve {
        defaultValue=0
        keys [
         {arriveTangent=-0.1713, interpMode=RCIM_Cubic, leaveTangent=-0.1713, tangentMode=RCTM_User, time=0.1, value=0}
         {arriveTangent=-0.0116, interpMode=RCIM_Cubic, leaveTangent=-0.0116, tangentMode=RCTM_User, time=0.35, value=0.9847}
         {arriveTangent=-0.0405, interpMode=RCIM_Cubic, leaveTangent=-0.0405, tangentMode=RCTM_User, time=1, value=0.25}
        ]
       }}}
     Scale Curve=&RandomRangeFloat(Minimum=0.35, Maximum=0.5)
    )) {isSetParameters=true}
   &SubUVAnimation(End Frame=11, Start Frame=8, SubUV Animation Mode:ENiagara_SubUVLookupMode=::NewEnumerator2)
   &AccelerationForce(
    Acceleration=&Lerp_Vector(A=<0 0 0>, B=<0 -500 0>, Alpha=$System.Timeline.NormalizedTime)
    Coordinate Space=::Simulation
   )
   &ParticleState()
   &ScaleSpriteSize(Scale Sprite Size Mode=::NewEnumerator0, Uniform Scale Factor=$Particles.Size Scale) {module=ScaleSpriteSize001}
   &SpriteRotationRate(Rotation Rate=&RandomRangeFloat(Minimum=-5, Maximum=5)) {module=SpriteRotationRate001}
   &Drag(Drag=1.5, Use Linear Drag=true, Use Rotational Drag=true)
   &CurlNoiseForce(Bias Noise=false, Noise Frequency=25, Noise Strength=200, Pan Noise=true, Pan Noise Field=<100 0 -500>) {module=CurlNoiseForce001}
   &SolveForcesAndVelocity(Clamp Velocity=false, Limit Acceleration=false)
   &ScaleColor(
    Scale Alpha=&Multiply_Float(
     A=&FloatFromCurve(FloatCurve=DataInterface() {propertyValues {Curve {keys [
          {interpMode=RCIM_Cubic, tangentMode=RCTM_User, time=0.35, value=0}
          {interpMode=RCIM_Cubic, tangentMode=RCTM_User, time=0.75, value=0.75}
          {tangentMode=RCTM_User, time=1, value=0}
         ]}}})
     B=1
    )
    Scale RGB=&Multiply_VectorByFloat(
     Vector=&MakeVectorFromLinearColorRGB(Color=&ColorFromCurve(ColorCurve=DataInterface() {propertyValues {
         AlphaCurve {keys []}
         BlueCurve {keys [<at 0 0.1667 RCIM_Cubic>,<at 0.3533 0>,<at 0.6474 0>]}
         GreenCurve {keys [<at 0 0.1667 RCIM_Cubic>,<at 0.3533 0>,<at 0.6474 0>]}
         RedCurve {keys [<at 0 0.1667 RCIM_Cubic>,<at 0.3533 0.25>,<at 0.6474 0.1>]}
        }}))
     Float=0.5
    )
    ScaleA=true
    ScaleRGB=true
   )
   &FN_Sprite_Attributes(
    Color=&MakeLinearColorFromVectorAndFloat(
     "Vector (RGB)"=&Lerp_Vector(
      A=&MakeVectorFromLinearColorRGB(Color=&ColorFromCurve(ColorCurve=DataInterface() {propertyValues {
          AlphaCurve {keys []}
          BlueCurve {keys [<at 0 0.492 RCIM_Cubic>,<at 1 0.1 RCIM_Cubic>]}
          GreenCurve {keys [<at 0 0.4055 RCIM_Cubic>,<at 1 0 RCIM_Cubic>]}
          RedCurve {keys [<at 0 3 RCIM_Cubic>,<at 1 3 RCIM_Cubic>]}
         }}))
      B=<0.05 0.05 0.05>
      Alpha=$User.Variant
     )
     "Float (Alpha)"=&Multiply_Float(
      A=&FloatFromCurve(
       FloatCurve=DataInterface() {propertyValues {Curve {keys [
           {interpMode=RCIM_Cubic, tangentMode=RCTM_User, time=0.1, value=0}
           {interpMode=RCIM_Cubic, tangentMode=RCTM_User, time=0.35, value=0.75}
           {tangentMode=RCTM_User, time=1, value=0}
          ]}}}
       Scale Curve=&Lerp_Float(A=1, B=2, Alpha=$User.Variant)
      )
      B=1
     )
    )
    Create Color Attribute=true
    Create SpriteSize Attribute=true
    SpriteSize=&Vector2DFromFloat(Value=&Multiply_Float(A=60, B=&Multiply_Float(A=$Particles.Random size, B=$Particles.Size Scale)))
    Write to Module Namespace=true
   )
   &FN_Sprite_Attributes(
    Color=&MakeLinearColorFromVectorAndFloat(
     "Vector (RGB)"=&MakeVectorFromLinearColorRGB(Color=&SelectLinearColorFromArray(
       Array Sampling Mode=::NewEnumerator1
       Direct Array Index=&MakeIntFromFloat(Float=$User.Variant)
       Color Selection Array=DataInterface() {propertyValues {ColorData [<rgba 1.5 0 0.05 0.5>, <rgba 0 0 0 1>]}}
       Random Seed=?
      ))
     "Float (Alpha)"=&Multiply_Float(
      A=&FN_FrontEndFloatModifier(Gameplay=0.2, Front End=0.1)
      B=&Multiply_Float(
       A=&FloatFromCurve(
        FloatCurve=DataInterface() {propertyValues {Curve {keys [
            {interpMode=RCIM_Cubic, tangentMode=RCTM_User, time=0.1, value=0}
            {interpMode=RCIM_Cubic, tangentMode=RCTM_User, time=0.35, value=0.75}
            {tangentMode=RCTM_User, time=1, value=0}
           ]}}}
        Scale Curve=&Lerp_Float(A=1, B=2, Alpha=$User.Variant)
       )
       B=1
      )
     )
    )
    Create Color Attribute=true
    Create SpriteSize Attribute=true
    SpriteSize=&Vector2DFromFloat(Value=&Multiply_Float(A=200, B=&Multiply_Float(A=$Particles.Random size, B=$Particles.Size Scale)))
    Write to Module Namespace=true
   ) {module=FN_Sprite_Attributes001}
   &SubUVAnimation(
    Direct Frame Index:NiagaraFloat=&RandomRangeFloat(Minimum=8, Maximum=11)
    "Sprite Renderer "=DataInterface() {propertyValues {SpriteRenderer=&self:Fire001_1.NiagaraSpriteRendererProperties_1}}
    SubUV Animation Mode:ENiagara_SubUVLookupModeV2=::NewEnumerator5
   ) {module=SubUVAnimation001}
  ]
  rendererClasses [&NiagaraSpriteRendererProperties]
  renderers [
   {rendererClass=&NiagaraSpriteRendererProperties}
   {rendererClass=&NiagaraSpriteRendererProperties, rendererIndex=1}
  ]
 }
 Flare {
  bEnabled=false
  emitterUpdateScript [
   &EmitterState(Life Cycle Mode=::NewEnumerator1, Loop Behavior=::NewEnumerator1, Loop Duration Mode=::NewEnumerator1, UseLoopDelay=false)
   &SpawnBurst_Instantaneous(
    Spawn Count=&SelectIntFromArray(
     Direct Array Index=&MakeIntFromFloat(Float=$User.Variant)
     Int Selection Array=DataInterface() {propertyValues {IntData [4,0]}}
    )
    Use Spawn Probability=false
    UseLoopCountLimit=false
   )
  ]
  particleSpawnScript [
   &InitializeParticle(
    Color=&MultiplyLinearColor(
     Linear Color=<rgba 1 1 1 1>
     Scale Factor=&MakeVector4FromVectorAndFloat(
      "Vector (RGB)"=&FN_FrontEndVector3Modifier(Gameplay=<1 1 1>, Front End=<1 1 0.65>)
      "Float (Alpha)"=1
     )
    )
    Lifetime=5
    Mesh Scale=&Multiply_VectorByFloat(
     Vector=&SelectVectorFromArray(
      Direct Array Index=&ReturnExecIndex()
      Vector Selection Array=DataInterface() {propertyValues {FloatData [<1 1 1>, <1 1 2>, <1 1 1>, <1 1 2>]}}
     )
     Float=2.2
    )
    Mesh Scale Mode=::NewEnumerator4
    Position Offset=$System.FlareOffset
    Sprite Size Mode=::NewEnumerator3
    Uniform Sprite Size=20
    UsePositionOffset=true
    Write Material Random=true
   )
   &CameraOffset(Camera Offset Amount=2)
   &InitialMeshOrientation(
    Mesh Axis To Orient=<0 1 0>
    Mesh Orientation Mode=::NewEnumerator2
    Orientation Vector=<0 0 1>
    Rotation=<0 0 -0.25>
    Use Rotation Vector=true
   )
  ]
  particleUpdateScript [
   &ParticleState(Kill Particles When Lifetime Has Elapsed=false, Let Infinitely Lived Particles Die When Emitter Deactivates=true)
   &InterpolateOverTime(
    Rate Of Change=4
    Target Float Value:NiagaraFloat=&RandomRangeFloat(Minimum=0.6, Maximum=1.4, Evaluation Type=::NewEnumerator1)
   ) {module=InterpolateOverTime002}
   &ScaleMeshSize(Scale Factor=&MakeVector(
     X=&FloatFromCurve(
      FloatCurve=DataInterface() {propertyValues {Curve {defaultValue=1, keys [{tangentMode=RCTM_User, time=0, value=0}, {tangentMode=RCTM_User, time=0.4, value=1}]}}}
      CurveIndex=$Particles.Age
     )
     Y=1
     Z=1
    )) {module=ScaleMeshSize001}
   &ScaleMeshSize(Scale Factor=&VectorFromFloat(Value=$Particles.InterpolateOverTime002.InterpolateOverTime_Float.Moving Average))
   &ScaleSpriteSize(Scale Sprite Size Mode=::NewEnumerator0, Uniform Scale Factor=$Particles.InterpolateOverTime002.InterpolateOverTime_Float.Moving Average)
   &InterpolateOverTime(
    Rate Of Change=2
    Target Float Value:NiagaraFloat=&RandomRangeFloat(Minimum=1, Maximum=1.35, Evaluation Type=::NewEnumerator1)
   ) {module=InterpolateOverTime003}
   &FN_Sprite_Attributes(
    Camera Offset:NiagaraFloat=&Add_Float(A=$Particles.CameraOffset, B=2)
    Color=<rgba 0.05 0.2 0.5 1>
    Create Camera Offset Attribute=true
    Create Color Attribute=true
    Create SpriteSize Attribute=true
    SpriteSize=&Multiply_Vector2D_ByFloat(Float=$Particles.InterpolateOverTime003.InterpolateOverTime_Float.Moving Average, Vector2D=$Particles.SpriteSize)
    Write to Module Namespace=true
   )
   /SetVariables(
    Particles.Sort:NiagaraInt32=&ReturnExecIndex()
    Particles.VisibilityTag=&ReturnExecIndex()
   ) {isSetParameters=true, module=SetVariables_#8}
   &FN_Mesh_Attributes(
    Create Scale Attribute=true
    Scale=&Multiply_VectorByFloat(Vector=&Multiply_Vector(A=$Particles.Scale, B=<1 1 1>), Float=0.8)
    Write to Module Namespace=true
   )
   /SetVariables(Particles.CameraOffset_02=6) {isSetParameters=true, module=SetVariables_#9}
   /SetVariables(Particles.DarkColor:LinearColor=&MakeLinearColorFromVectorAndFloat(
     "Vector (RGB)"=&FN_FrontEndVector3Modifier(Gameplay=<1 1 1>, Front End=<0.25 0.25 0.25>)
     "Float (Alpha)"=1
    )) {isSetParameters=true}
  ]
  rendererClasses [&NiagaraMeshRendererProperties]
  renderers [
   {rendererClass=&NiagaraMeshRendererProperties, rendererIndex=0}
   {rendererClass=&NiagaraMeshRendererProperties, rendererIndex=1}
   {rendererClass=&NiagaraMeshRendererProperties, rendererIndex=2}
   {rendererClass=&NiagaraMeshRendererProperties, rendererIndex=3}
   {rendererClass=&NiagaraMeshRendererProperties, rendererIndex=4}
   {rendererClass=&NiagaraMeshRendererProperties, rendererIndex=5}
  ]
 }
 FlareVariant {
  emitterUpdateScript [
   &EmitterState(Life Cycle Mode=::NewEnumerator1, Loop Behavior=::NewEnumerator1, Loop Duration Mode=::NewEnumerator1, UseLoopDelay=false)
   &SpawnBurst_Instantaneous(
    Spawn Count=&Multiply_Int(
     A=&SelectIntFromArray(
      Direct Array Index=&MakeIntFromFloat(Float=$User.Variant)
      Int Selection Array=DataInterface() {propertyValues {IntData [0,4]}}
     )
     B=&MakeIntFromFloat(Float=&FN_FrontEndFloatModifier(Gameplay=1, Front End=0))
    )
    Use Spawn Probability=false
    UseLoopCountLimit=false
   )
  ]
  particleSpawnScript [
   &InitializeParticle(
    Color=&MultiplyLinearColor(Linear Color=<rgba 1 1 1 1>, Scale Factor=<1 1 1 1>)
    Lifetime=5
    Mesh Scale=&Multiply_VectorByFloat(
     Vector=&SelectVectorFromArray(
      Direct Array Index=&ReturnExecIndex()
      Vector Selection Array=DataInterface() {propertyValues {FloatData [<1 1 1>, <1 1 2>, <1 1 1>, <1 1 2>]}}
     )
     Float=2.2
    )
    Mesh Scale Mode=::NewEnumerator4
    Position Offset=$System.FlareOffset
    Sprite Size Mode=::NewEnumerator3
    Uniform Sprite Size=20
    UsePositionOffset=true
    Write Material Random=true
   )
   &CameraOffset(Camera Offset Amount=2)
   &InitialMeshOrientation(
    Mesh Axis To Orient=<0 1 0>
    Mesh Orientation Mode=::NewEnumerator2
    Orientation Vector=<0 0 1>
    Rotation=<0 0 -0.25>
    Use Rotation Vector=true
   )
  ]
  particleUpdateScript [
   &ParticleState(Kill Particles When Lifetime Has Elapsed=false, Let Infinitely Lived Particles Die When Emitter Deactivates=true)
   &InterpolateOverTime(
    Rate Of Change=4
    Target Float Value:NiagaraFloat=&RandomRangeFloat(Minimum=0.6, Maximum=1.4, Evaluation Type=::NewEnumerator1)
   ) {module=InterpolateOverTime002}
   &ScaleMeshSize(Scale Factor=&MakeVector(
     X=&FloatFromCurve(
      FloatCurve=DataInterface() {propertyValues {Curve {defaultValue=1, keys [{tangentMode=RCTM_User, time=0, value=0}, {tangentMode=RCTM_User, time=0.4, value=1}]}}}
      CurveIndex=$Particles.Age
     )
     Y=1
     Z=1
    )) {module=ScaleMeshSize001}
   &ScaleMeshSize(Scale Factor=&VectorFromFloat(Value=$Particles.InterpolateOverTime002.InterpolateOverTime_Float.Moving Average))
   &ScaleSpriteSize(Scale Sprite Size Mode=::NewEnumerator0, Uniform Scale Factor=$Particles.InterpolateOverTime002.InterpolateOverTime_Float.Moving Average)
   &InterpolateOverTime(
    Rate Of Change=2
    Target Float Value:NiagaraFloat=&RandomRangeFloat(Minimum=1, Maximum=1.35, Evaluation Type=::NewEnumerator1)
   ) {module=InterpolateOverTime003}
   &FN_Sprite_Attributes(
    Camera Offset:NiagaraFloat=&Add_Float(A=$Particles.CameraOffset, B=2)
    Color=<rgba 0.05 0.2 0.5 1>
    Create Camera Offset Attribute=true
    Create Color Attribute=true
    Create SpriteSize Attribute=true
    SpriteSize=&Multiply_Vector2D_ByFloat(Float=$Particles.InterpolateOverTime003.InterpolateOverTime_Float.Moving Average, Vector2D=$Particles.SpriteSize)
    Write to Module Namespace=true
   )
   /SetVariables(
    Particles.Sort:NiagaraInt32=&ReturnExecIndex()
    Particles.VisibilityTag=&ReturnExecIndex()
   ) {isSetParameters=true, module=SetVariables_#8}
   &FN_Mesh_Attributes(
    Create Scale Attribute=true
    Scale=&Multiply_VectorByFloat(Vector=&Multiply_Vector(A=$Particles.Scale, B=<1 1 1>), Float=0.8)
    Write to Module Namespace=true
   )
   /SetVariables(Particles.CameraOffset_02=6) {isSetParameters=true, module=SetVariables_#9}
   /SetVariables(Particles.DarkColor:LinearColor=&MultiplyLinearColor(Linear Color=<rgba 1 1 1 1>, Scale Factor=<0.5 0.5 0.5 1>)) {isSetParameters=true}
  ]
  rendererClasses [&NiagaraMeshRendererProperties]
  renderers [
   {rendererClass=&NiagaraMeshRendererProperties, rendererIndex=0}
   {rendererClass=&NiagaraMeshRendererProperties, rendererIndex=1}
   {rendererClass=&NiagaraMeshRendererProperties, rendererIndex=2}
   {rendererClass=&NiagaraMeshRendererProperties, rendererIndex=3}
   {rendererClass=&NiagaraMeshRendererProperties, rendererIndex=4}
   {rendererClass=&NiagaraMeshRendererProperties, rendererIndex=5}
  ]
 }
 FlareVariant_FE {
  emitterUpdateScript [
   &EmitterState(Life Cycle Mode=::NewEnumerator1, Loop Behavior=::NewEnumerator1, Loop Duration Mode=::NewEnumerator1, UseLoopDelay=false)
   &SpawnBurst_Instantaneous(
    Spawn Count=&Multiply_Int(
     A=&SelectIntFromArray(
      Direct Array Index=&MakeIntFromFloat(Float=$User.Variant)
      Int Selection Array=DataInterface() {propertyValues {IntData [0,4]}}
     )
     B=&MakeIntFromFloat(Float=&FN_FrontEndFloatModifier(Gameplay=0, Front End=1))
    )
    Use Spawn Probability=false
    UseLoopCountLimit=false
   )
  ]
  particleSpawnScript [
   &InitializeParticle(
    Color=&MultiplyLinearColor(
     Linear Color=<rgba 1 1 1 1>
     Scale Factor=&MakeVector4FromVectorAndFloat(
      "Vector (RGB)"=&FN_FrontEndVector3Modifier(Gameplay=<1 1 1>, Front End=<1.5 0.5 0.5>)
      "Float (Alpha)"=1
     )
    )
    Lifetime=5
    Mesh Scale=&Multiply_VectorByFloat(
     Vector=&SelectVectorFromArray(
      Direct Array Index=&ReturnExecIndex()
      Vector Selection Array=DataInterface() {propertyValues {FloatData [<1 1 1>, <1 1 2>, <1 1 1>, <1 1 2>]}}
     )
     Float=2.2
    )
    Mesh Scale Mode=::NewEnumerator4
    Position Offset=$System.FlareOffset
    Sprite Size Mode=::NewEnumerator3
    Uniform Sprite Size=20
    UsePositionOffset=true
    Write Material Random=true
   )
   &CameraOffset(Camera Offset Amount=2)
   &InitialMeshOrientation(
    Mesh Axis To Orient=<0 1 0>
    Mesh Orientation Mode=::NewEnumerator2
    Orientation Vector=<0 0 1>
    Rotation=<0 0 -0.25>
    Use Rotation Vector=true
   )
  ]
  particleUpdateScript [
   &ParticleState(Kill Particles When Lifetime Has Elapsed=false, Let Infinitely Lived Particles Die When Emitter Deactivates=true)
   &InterpolateOverTime(
    Rate Of Change=4
    Target Float Value:NiagaraFloat=&RandomRangeFloat(Minimum=0.6, Maximum=1.4, Evaluation Type=::NewEnumerator1)
   ) {module=InterpolateOverTime002}
   &ScaleMeshSize(Scale Factor=&MakeVector(
     X=&FloatFromCurve(
      FloatCurve=DataInterface() {propertyValues {Curve {defaultValue=1, keys [{tangentMode=RCTM_User, time=0, value=0}, {tangentMode=RCTM_User, time=0.4, value=1}]}}}
      CurveIndex=$Particles.Age
     )
     Y=1
     Z=1
    )) {module=ScaleMeshSize001}
   &ScaleMeshSize(Scale Factor=&VectorFromFloat(Value=$Particles.InterpolateOverTime002.InterpolateOverTime_Float.Moving Average))
   &ScaleSpriteSize(Scale Sprite Size Mode=::NewEnumerator0, Uniform Scale Factor=$Particles.InterpolateOverTime002.InterpolateOverTime_Float.Moving Average)
   &InterpolateOverTime(
    Rate Of Change=2
    Target Float Value:NiagaraFloat=&RandomRangeFloat(Minimum=1, Maximum=1.35, Evaluation Type=::NewEnumerator1)
   ) {module=InterpolateOverTime003}
   &FN_Sprite_Attributes(
    Camera Offset:NiagaraFloat=&Add_Float(A=$Particles.CameraOffset, B=2)
    Color=<rgba 0.05 0.2 0.5 1>
    Create Camera Offset Attribute=true
    Create Color Attribute=true
    Create SpriteSize Attribute=true
    SpriteSize=&Multiply_Vector2D_ByFloat(Float=$Particles.InterpolateOverTime003.InterpolateOverTime_Float.Moving Average, Vector2D=$Particles.SpriteSize)
    Write to Module Namespace=true
   )
   /SetVariables(
    Particles.Sort:NiagaraInt32=&ReturnExecIndex()
    Particles.VisibilityTag=&ReturnExecIndex()
   ) {isSetParameters=true, module=SetVariables_#8}
   &FN_Mesh_Attributes(
    Create Scale Attribute=true
    Scale=&Multiply_VectorByFloat(Vector=&Multiply_Vector(A=$Particles.Scale, B=<1 1 1>), Float=0.8)
    Write to Module Namespace=true
   )
   /SetVariables(Particles.CameraOffset_02=6) {isSetParameters=true, module=SetVariables_#9}
   /SetVariables(Particles.DarkColor:LinearColor=&MakeLinearColorFromVectorAndFloat(
     "Vector (RGB)"=&FN_FrontEndVector3Modifier(Gameplay=<1 1 1>, Front End=<0.25 0.25 0.25>)
     "Float (Alpha)"=1
    )) {isSetParameters=true}
  ]
  rendererClasses [&NiagaraMeshRendererProperties]
  renderers [
   {rendererClass=&NiagaraMeshRendererProperties, rendererIndex=0}
   {rendererClass=&NiagaraMeshRendererProperties, rendererIndex=1}
   {rendererClass=&NiagaraMeshRendererProperties, rendererIndex=2}
   {rendererClass=&NiagaraMeshRendererProperties, rendererIndex=3}
   {rendererClass=&NiagaraMeshRendererProperties, rendererIndex=4}
   {rendererClass=&NiagaraMeshRendererProperties, rendererIndex=5}
  ]
 }
 Sparks {
  simTarget=GPUComputeSim
  emitterUpdateScript [
   &EmitterState(Life Cycle Mode=::NewEnumerator1, Loop Behavior=::NewEnumerator1, Loop Delay=2, Loop Duration Mode=::NewEnumerator1, UseLoopDelay=true)
   &SpawnRate(SpawnRate=&Lerp_Float(A=7, B=15, Alpha=$System.Timeline.NormalizedTime), Use Spawn Probability=false)
  ]
  particleSpawnScript [
   &InitializeParticle(
    Color=&SelectLinearColorFromArray(
     Array Sampling Mode=::NewEnumerator1
     Direct Array Index=&MakeIntFromFloat(Float=$User.Variant)
     Color Selection Array=DataInterface() {propertyValues {ColorData [<rgba 5 0 0 1>, <rgba 0.01 0.01 0.01 2>]}}
     Random Seed=?
    )
    Lifetime Max=1.7
    Lifetime Min=1.25
    Lifetime Mode=::NewEnumerator1
    Position Offset=$System.FlareOffset
    Sprite Size=&Lerp_Vector2(A=<2 20>, B=<4 20>, Alpha=$User.Variant)
    Sprite Size Mode=::NewEnumerator4
    UsePositionOffset=true
    Write Material Random=true
   )
   &ShapeLocation(MaskSpawnGroup=false, Override Seed=false, Sphere Radius=1, UseOwnerScale=true)
   &AddVelocity(
    Rotation Coordinate Space=::Simulation
    Velocity=&FN_FrontEndVector3Modifier(Gameplay=<-1 0 0>, Front End=<-1 -0.25 0>)
    Velocity Speed Scale=150
   ) {module=AddVelocity001}
   &CameraOffset() {module=CameraOffset001}
  ]
  particleUpdateScript [
   &ParticleState()
   /SetVariables(Particles.Glow03_Color=&MakeLinearColorFromVectorAndFloat(
     "Vector (RGB)"=&MakeVectorFromLinearColorRGB(Color=&MultiplyLinearColor(Linear Color=<rgba 1.5 0 0.05 1>, Scale Factor=&MakeVector4_FromFloat(Float=1)))
     "Float (Alpha)"=&Multiply_Float(A=0.2, B=1)
    )) {isSetParameters=true}
   /SetVariables(Particles.Glow03_Size=&Multiply_Vector2D_ByFloat(
     Float=&FloatFromCurve(
      FloatCurve=DataInterface() {propertyValues {Curve {keys [
          <at 0 0 RCIM_Cubic>
          <at 0.2 1.3 RCIM_Cubic>
          {arriveTangent=-0.1423, interpMode=RCIM_Cubic, leaveTangent=-0.1423, tangentMode=RCTM_User, time=1, value=0}
         ]}}}
      Scale Curve=10
     )
     Vector2D=<0.6 1.2>
    )) {isSetParameters=true}
   &AccelerationForce(
    Acceleration=&Lerp_Vector(A=<0 0 0>, B=<0 -500 0>, Alpha=$System.Timeline.NormalizedTime)
    Coordinate Space=::Simulation
   )
   &ScaleColor(
    Scale Alpha=&FloatFromCurve(FloatCurve=DataInterface() {propertyValues {Curve {keys [<at 0.3 0>,<at 0.4 1>,<at 1 1>]}}})
    ScaleA=true
    ScaleRGB=true
   )
   &ScaleSpriteSize(Scale Factor=&Vector2DFromFloat(Value=&FloatFromCurve(
      FloatCurve=DataInterface() {propertyValues {Curve {keys [
          <at 0 1 RCIM_Cubic>
          {interpMode=RCIM_Cubic, leaveTangent=-3.0296, tangentMode=RCTM_Break, time=0.2, value=1.2}
          {arriveTangent=-1.8107, interpMode=RCIM_Cubic, leaveTangent=-1.8107, tangentMode=RCTM_User, time=1, value=0}
         ]}}}
      Scale Curve=&Multiply_Float(
       A=&SelectFloatFromArray(
        Direct Array Index=&MakeIntFromFloat(Float=$User.Variant)
        Float Selection Array=DataInterface() {propertyValues {FloatData [0.4,0.6]}}
       )
       B=&ClampFloat(
        Float=&FN_LOD_Distance_Remap(Start Distance=0, End Distance=500, Minimum Output=0, Maximum Output=1, Clamp Output=true)
        Min=0.2
       )
      )
     ))) {module=ScaleSpriteSize001}
   &Drag(Use Linear Drag=true, Use Rotational Drag=true)
   &CurlNoiseForce(Bias Noise=false, Noise Frequency=100, Noise Strength=250, Pan Noise=true, Pan Noise Field=<100 0 -500>) {module=CurlNoiseForce001}
   &SolveForcesAndVelocity(Clamp Velocity=false, Limit Acceleration=false)
  ]
  rendererClasses [&NiagaraSpriteRendererProperties]
  renderers [{rendererClass=&NiagaraSpriteRendererProperties}]
 }
 TopSmoke {
  bEnabled=false
  emitterUpdateScript [
   &EmitterState(Life Cycle Mode=::NewEnumerator1, Loop Behavior=::NewEnumerator1, Loop Duration Mode=::NewEnumerator1, UseLoopDelay=false)
   &SpawnBurst_Instantaneous(
    Spawn Count=&SelectIntFromArray(
     Direct Array Index=&MakeIntFromFloat(Float=$User.Variant)
     Int Selection Array=DataInterface() {propertyValues {IntData [1,0]}}
    )
    Use Spawn Probability=false
    UseLoopCountLimit=false
   )
  ]
  particleSpawnScript [
   &InitializeParticle(
    Color=<rgba 1 0.2 0.35 1>
    Lifetime=5
    Mesh Scale=&Multiply_VectorByFloat(Vector=<1 0.35 1>, Float=1.2)
    Mesh Scale Mode=::NewEnumerator4
    Position Offset=&Add_Vector(A=$System.FlareOffset, B=<0 -0.5 0>)
    UsePositionOffset=true
    Write Material Random=true
   )
   &InitialMeshOrientation(Mesh Orientation Mode=::NewEnumerator2, Use Rotation Vector=false)
   &CameraOffset(Camera Offset Amount=4)
   /SetVariables(Particles.Color1:LinearColor=&Multiply_LinearColorByFloat(
     Linear Color=&Multiply_LinearColorByFloat(
      Linear Color=&MultiplyLinearColor(
       Linear Color=<rgba 1 1 1 1>
       Scale Factor=&MakeVector4FromVectorAndFloat(
        "Vector (RGB)"=&FN_FrontEndVector3Modifier(Gameplay=<1 1 1>, Front End=<0.25 0.25 0.25>)
        "Float (Alpha)"=1
       )
      )
      Float=1
     )
     Float=1
    )) {isSetParameters=true, module=SetVariables_#3}
   /SetVariables(Particles.Color2=&Multiply_LinearColorByFloat(Linear Color=<rgba 1 1 1 3>, Float=1)) {isSetParameters=true, module=SetVariables_#4}
   /SetVariables(Particles.Color3:LinearColor=&Multiply_LinearColorByFloat(Linear Color=<rgba 1 1 1 3>, Float=0.3)) {isSetParameters=true, module=SetVariables_#5}
  ]
  particleUpdateScript [
   &ParticleState(Kill Particles When Lifetime Has Elapsed=false)
   /SetVariables(Particles.MeshOrientation=&MakeQuatFromAxisAngle(
     Axis=<0 1 0>
     AngleInDegrees=&Lerp_Float(
      A=110
      B=70
      Alpha=&RemapRange(
       Input Value To Remap=&Waveform(
        Waveform Count=::NewEnumerator1
        "Waveform [1]"=::NewEnumerator0
        "[1] Frequency"=0.2
        "[1] Amplitude Min/Max"=<-1 1>
        "Waveform [2] Blend"=::NewEnumerator2
        "Waveform [2]"=::NewEnumerator1
        "[2] Input Position"=0
        "[2] Phase"=$Emitter.Age
        "[2] Phase Scale"=1
        "[2] Compound Frequency"=2
        "[2] Frequency"=0.1
        "[2] Amplitude Min/Max"=<0.5 1>
        "[2] Scale"=1
        "[2] Bias"=0
        "[2] Exponent"=1
        "[2] Pulse Duration"=0.1
        "[2] Clip Triangle Min/Max"=<0 0>
        "[2] Random Seed"=0
       )
       Input Min=-1
       Input Max=1
       Output Min=0
       Output Max=1
      )
     )
    )) {isSetParameters=true, module=SetVariables_#6}
   &InterpolateOverTime(
    Rate Of Change=6
    Target Vector Value:Vector3f=&ConvertPositionToVector(
     Input Position=&TransformPosition(Position=$Particles.Position, Source Space=::Local, Destination Space=::World)
     Convert To Absolute Large World Vector=false
    )
    Variable Type=::NewEnumerator2
   )
   /SetVariables(Particles.Interpolate Dynamic=&MakeFloat4FromVectorAndFloat("Vector (RGB)"=$StackContext.InterpolateOverTime.InterpolateOverTime_V3_.Moving Average, "Float (Alpha)"=0.3)) {isSetParameters=true, module=SetVariables_#7}
   &ScaleColor(ScaleA=true, ScaleRGB=false)
   &DynamicMaterialParameters(
    Index 0 Param 2=&FloatFromCurve(
     FloatCurve=DataInterface() {propertyValues {Curve {defaultValue=1, keys [{tangentMode=RCTM_User, time=1.5, value=0}, {tangentMode=RCTM_User, time=2.5, value=1}]}}}
     CurveIndex=$Particles.Age
    )
    Index 0 Param 4=&Frac_Float(Float=$System.FN_Integrator002.Value)
    Module.Index0Param0Write=false
    Module.Index0Param1Write=true
    Module.Index0Param2Write=false
    Module.Index0Param3Write=true
   )
  ]
  rendererClasses [&NiagaraMeshRendererProperties]
  renderers [
   {rendererClass=&NiagaraMeshRendererProperties, rendererIndex=0}
   {rendererClass=&NiagaraMeshRendererProperties, rendererIndex=1}
   {rendererClass=&NiagaraMeshRendererProperties, rendererIndex=2}
   {rendererClass=&NiagaraMeshRendererProperties, rendererIndex=3}
  ]
 }
 TopSmokeVariant {
  emitterUpdateScript [
   &EmitterState(Life Cycle Mode=::NewEnumerator1, Loop Behavior=::NewEnumerator1, Loop Duration Mode=::NewEnumerator1, UseLoopDelay=false)
   &SpawnBurst_Instantaneous(
    Spawn Count=&MakeIntFromFloat(Float=&FN_FrontEndFloatModifier(
      Gameplay=&MakeFloatFromInt(Int=&SelectIntFromArray(
        Direct Array Index=&MakeIntFromFloat(Float=$User.Variant)
        Int Selection Array=DataInterface() {propertyValues {IntData [0,1]}}
       ))
      Front End=0
     ))
    Use Spawn Probability=false
    UseLoopCountLimit=false
   )
  ]
  particleSpawnScript [
   &InitializeParticle(
    Color=&SelectLinearColorFromArray(
     Array Sampling Mode=::NewEnumerator1
     Direct Array Index=&MakeIntFromFloat(Float=$User.Variant)
     Color Selection Array=DataInterface() {propertyValues {ColorData [<rgba 1 0.2 0.35 1>, <rgba 0.2188 0.2188 0.2188 2>]}}
     Random Seed=?
    )
    Lifetime=5
    Mesh Scale=&Multiply_VectorByFloat(Vector=<1 0.35 1>, Float=1.2)
    Mesh Scale Mode=::NewEnumerator4
    Position Offset=&Add_Vector(A=$System.FlareOffset, B=<0 -0.5 0>)
    UsePositionOffset=true
    Write Material Random=true
   )
   &InitialMeshOrientation(Mesh Orientation Mode=::NewEnumerator2, Use Rotation Vector=false)
   &CameraOffset(Camera Offset Amount=4)
   /SetVariables(Particles.Color1:LinearColor=&Multiply_LinearColorByFloat(
     Linear Color=&MultiplyLinearColor(
      Linear Color=<rgba 1 1 1 1>
      Scale Factor=&MakeVector4FromVectorAndFloat(
       "Vector (RGB)"=&FN_FrontEndVector3Modifier(Gameplay=<1 1 1>, Front End=<0.5 1 6>)
       "Float (Alpha)"=1
      )
     )
     Float=&FN_FrontEndFloatModifier(
      Gameplay=&SelectFloatFromArray(
       Direct Array Index=&MakeIntFromFloat(Float=$User.Variant)
       Float Selection Array=DataInterface() {propertyValues {FloatData [1,0.4]}}
      )
      Front End=&SelectFloatFromArray(
       Direct Array Index=&MakeIntFromFloat(Float=$User.Variant)
       Float Selection Array=DataInterface() {propertyValues {FloatData [1,1]}}
      )
     )
    )) {isSetParameters=true, module=SetVariables_#3}
   /SetVariables(Particles.Color2=&Multiply_LinearColorByFloat(
     Linear Color=&MultiplyLinearColor(
      Linear Color=<rgba 1 1 1 1>
      Scale Factor=&MakeVector4FromVectorAndFloat(
       "Vector (RGB)"=&FN_FrontEndVector3Modifier(Gameplay=<1 1 1>, Front End=<0.7 1 6>)
       "Float (Alpha)"=1
      )
     )
     Float=&FN_FrontEndFloatModifier(
      Gameplay=1
      Front End=&SelectFloatFromArray(
       Direct Array Index=&MakeIntFromFloat(Float=$User.Variant)
       Float Selection Array=DataInterface() {propertyValues {FloatData [1,2]}}
      )
     )
    )) {isSetParameters=true, module=SetVariables_#4}
   /SetVariables(Particles.Color3:LinearColor=&Multiply_LinearColorByFloat(
     Linear Color=<rgba 0.2448 0.2448 0.2448 3>
     Float=&SelectFloatFromArray(
      Direct Array Index=&MakeIntFromFloat(Float=$User.Variant)
      Float Selection Array=DataInterface() {propertyValues {FloatData [0.3,1]}}
     )
    )) {isSetParameters=true, module=SetVariables_#5}
  ]
  particleUpdateScript [
   &ParticleState(Kill Particles When Lifetime Has Elapsed=false)
   /SetVariables(Particles.MeshOrientation=&MakeQuatFromAxisAngle(
     Axis=<0 1 0>
     AngleInDegrees=&Lerp_Float(
      A=110
      B=70
      Alpha=&RemapRange(
       Input Value To Remap=&Waveform(
        Waveform Count=::NewEnumerator1
        "Waveform [1]"=::NewEnumerator0
        "[1] Frequency"=0.2
        "[1] Amplitude Min/Max"=<-1 1>
        "Waveform [2] Blend"=::NewEnumerator2
        "Waveform [2]"=::NewEnumerator1
        "[2] Input Position"=0
        "[2] Phase"=$Emitter.Age
        "[2] Phase Scale"=1
        "[2] Compound Frequency"=2
        "[2] Frequency"=0.1
        "[2] Amplitude Min/Max"=<0.5 1>
        "[2] Scale"=1
        "[2] Bias"=0
        "[2] Exponent"=1
        "[2] Pulse Duration"=0.1
        "[2] Clip Triangle Min/Max"=<0 0>
        "[2] Random Seed"=0
       )
       Input Min=-1
       Input Max=1
       Output Min=0
       Output Max=1
      )
     )
    )) {isSetParameters=true, module=SetVariables_#6}
   &InterpolateOverTime(
    Rate Of Change=6
    Target Vector Value:Vector3f=&ConvertPositionToVector(
     Input Position=&TransformPosition(Position=$Particles.Position, Source Space=::Local, Destination Space=::World)
     Convert To Absolute Large World Vector=false
    )
    Variable Type=::NewEnumerator2
   )
   /SetVariables(Particles.Interpolate Dynamic=&MakeFloat4FromVectorAndFloat("Vector (RGB)"=$StackContext.InterpolateOverTime.InterpolateOverTime_V3_.Moving Average, "Float (Alpha)"=0.3)) {isSetParameters=true, module=SetVariables_#7}
   &ScaleColor(ScaleA=true, ScaleRGB=false)
   &DynamicMaterialParameters(
    Index 0 Param 2=&FloatFromCurve(
     FloatCurve=DataInterface() {propertyValues {Curve {defaultValue=1, keys [{tangentMode=RCTM_User, time=1.5, value=0}, {tangentMode=RCTM_User, time=2.5, value=1}]}}}
     CurveIndex=$Particles.Age
    )
    Index 0 Param 4=&Frac_Float(Float=$System.FN_Integrator002.Value)
    Module.Index0Param0Write=false
    Module.Index0Param1Write=true
    Module.Index0Param2Write=false
    Module.Index0Param3Write=true
   )
  ]
  rendererClasses [&NiagaraMeshRendererProperties]
  renderers [
   {rendererClass=&NiagaraMeshRendererProperties, rendererIndex=0}
   {rendererClass=&NiagaraMeshRendererProperties, rendererIndex=1}
   {rendererClass=&NiagaraMeshRendererProperties, rendererIndex=2}
   {rendererClass=&NiagaraMeshRendererProperties, rendererIndex=3}
  ]
 }
 TopSmokeVariant_FE {
  emitterUpdateScript [
   &EmitterState(Life Cycle Mode=::NewEnumerator1, Loop Behavior=::NewEnumerator1, Loop Duration Mode=::NewEnumerator1, UseLoopDelay=false)
   &SpawnBurst_Instantaneous(
    Spawn Count=&MakeIntFromFloat(Float=&FN_FrontEndFloatModifier(
      Gameplay=0
      Front End=&MakeFloatFromInt(Int=&SelectIntFromArray(
        Direct Array Index=&MakeIntFromFloat(Float=$User.Variant)
        Int Selection Array=DataInterface() {propertyValues {IntData [0,1]}}
       ))
     ))
    Use Spawn Probability=false
    UseLoopCountLimit=false
   )
  ]
  particleSpawnScript [
   &InitializeParticle(
    Color=&SelectLinearColorFromArray(
     Array Sampling Mode=::NewEnumerator1
     Direct Array Index=&MakeIntFromFloat(Float=$User.Variant)
     Color Selection Array=DataInterface() {propertyValues {ColorData [<rgba 1 0.2 0.35 1>, <rgba 0.05 0.05 0.05 2>]}}
     Random Seed=?
    )
    Lifetime=5
    Mesh Scale=&Multiply_VectorByFloat(Vector=<1 0.35 1>, Float=1.2)
    Mesh Scale Mode=::NewEnumerator4
    Position Offset=&Add_Vector(A=$System.FlareOffset, B=<0 -0.5 0>)
    UsePositionOffset=true
    Write Material Random=true
   )
   &InitialMeshOrientation(Mesh Orientation Mode=::NewEnumerator2, Use Rotation Vector=false)
   &CameraOffset(Camera Offset Amount=4)
   /SetVariables(Particles.Color1:LinearColor=&Multiply_LinearColorByFloat(
     Linear Color=&MultiplyLinearColor(
      Linear Color=<rgba 1 1 1 1>
      Scale Factor=&MakeVector4FromVectorAndFloat(
       "Vector (RGB)"=&FN_FrontEndVector3Modifier(Gameplay=<1 1 1>, Front End=<1.5 1 0.8>)
       "Float (Alpha)"=1
      )
     )
     Float=&FN_FrontEndFloatModifier(
      Gameplay=&SelectFloatFromArray(
       Direct Array Index=&MakeIntFromFloat(Float=$User.Variant)
       Float Selection Array=DataInterface() {propertyValues {FloatData [1,0.4]}}
      )
      Front End=&SelectFloatFromArray(
       Direct Array Index=&MakeIntFromFloat(Float=$User.Variant)
       Float Selection Array=DataInterface() {propertyValues {FloatData [1,0.4]}}
      )
     )
    )) {isSetParameters=true, module=SetVariables_#3}
   /SetVariables(Particles.Color2=&Multiply_LinearColorByFloat(
     Linear Color=&MultiplyLinearColor(
      Linear Color=<rgba 1 1 1 1>
      Scale Factor=&MakeVector4FromVectorAndFloat(
       "Vector (RGB)"=&FN_FrontEndVector3Modifier(Gameplay=<1 1 1>, Front End=<1.5 1 0.8>)
       "Float (Alpha)"=1
      )
     )
     Float=&FN_FrontEndFloatModifier(
      Gameplay=1
      Front End=&SelectFloatFromArray(
       Direct Array Index=&MakeIntFromFloat(Float=$User.Variant)
       Float Selection Array=DataInterface() {propertyValues {FloatData [1,1]}}
      )
     )
    )) {isSetParameters=true, module=SetVariables_#4}
   /SetVariables(Particles.Color3:LinearColor=&Multiply_LinearColorByFloat(
     Linear Color=<rgba 0.2448 0.2448 0.2448 3>
     Float=&SelectFloatFromArray(
      Direct Array Index=&MakeIntFromFloat(Float=$User.Variant)
      Float Selection Array=DataInterface() {propertyValues {FloatData [0.3,1]}}
     )
    )) {isSetParameters=true, module=SetVariables_#5}
  ]
  particleUpdateScript [
   &ParticleState(Kill Particles When Lifetime Has Elapsed=false)
   /SetVariables(Particles.MeshOrientation=&MakeQuatFromAxisAngle(
     Axis=<0 1 0>
     AngleInDegrees=&Lerp_Float(
      A=110
      B=70
      Alpha=&RemapRange(
       Input Value To Remap=&Waveform(
        Waveform Count=::NewEnumerator1
        "Waveform [1]"=::NewEnumerator0
        "[1] Frequency"=0.2
        "[1] Amplitude Min/Max"=<-1 1>
        "Waveform [2] Blend"=::NewEnumerator2
        "Waveform [2]"=::NewEnumerator1
        "[2] Input Position"=0
        "[2] Phase"=$Emitter.Age
        "[2] Phase Scale"=1
        "[2] Compound Frequency"=2
        "[2] Frequency"=0.1
        "[2] Amplitude Min/Max"=<0.5 1>
        "[2] Scale"=1
        "[2] Bias"=0
        "[2] Exponent"=1
        "[2] Pulse Duration"=0.1
        "[2] Clip Triangle Min/Max"=<0 0>
        "[2] Random Seed"=0
       )
       Input Min=-1
       Input Max=1
       Output Min=0
       Output Max=1
      )
     )
    )) {isSetParameters=true, module=SetVariables_#6}
   &InterpolateOverTime(
    Rate Of Change=6
    Target Vector Value:Vector3f=&ConvertPositionToVector(
     Input Position=&TransformPosition(Position=$Particles.Position, Source Space=::Local, Destination Space=::World)
     Convert To Absolute Large World Vector=false
    )
    Variable Type=::NewEnumerator2
   )
   /SetVariables(Particles.Interpolate Dynamic=&MakeFloat4FromVectorAndFloat("Vector (RGB)"=$StackContext.InterpolateOverTime.InterpolateOverTime_V3_.Moving Average, "Float (Alpha)"=0.3)) {isSetParameters=true, module=SetVariables_#7}
   &ScaleColor(ScaleA=true, ScaleRGB=false)
   &DynamicMaterialParameters(
    Index 0 Param 2=&FloatFromCurve(
     FloatCurve=DataInterface() {propertyValues {Curve {defaultValue=1, keys [{tangentMode=RCTM_User, time=1.5, value=0}, {tangentMode=RCTM_User, time=2.5, value=1}]}}}
     CurveIndex=$Particles.Age
    )
    Index 0 Param 4=&Frac_Float(Float=$System.FN_Integrator002.Value)
    Module.Index0Param0Write=false
    Module.Index0Param1Write=true
    Module.Index0Param2Write=false
    Module.Index0Param3Write=true
   )
  ]
  rendererClasses [&NiagaraMeshRendererProperties]
  renderers [
   {rendererClass=&NiagaraMeshRendererProperties, rendererIndex=0}
   {rendererClass=&NiagaraMeshRendererProperties, rendererIndex=1}
   {rendererClass=&NiagaraMeshRendererProperties, rendererIndex=2}
   {rendererClass=&NiagaraMeshRendererProperties, rendererIndex=3}
  ]
 }
}
summary {
 systemName=NS_Glider_MistRaven_Candy_Idle
 emitters [
  {emitterName=Feathers, rendererClasses [&NiagaraMeshRendererProperties]}
  {bEnabled=false, emitterName=Flare, rendererClasses [&NiagaraMeshRendererProperties]}
  {emitterName=Airglow, rendererClasses [&NiagaraSpriteRendererProperties]}
  {bEnabled=false, emitterName=TopSmoke, rendererClasses [&NiagaraMeshRendererProperties]}
  {bEnabled=false, emitterName=BottomSmoke, rendererClasses [&NiagaraMeshRendererProperties]}
  {emitterName=Fire001, simTarget=GPUComputeSim, rendererClasses [&NiagaraSpriteRendererProperties]}
  {emitterName=Sparks, simTarget=GPUComputeSim, rendererClasses [&NiagaraSpriteRendererProperties]}
  {emitterName=BottomSmokeVariant, rendererClasses [&NiagaraMeshRendererProperties]}
  {emitterName=TopSmokeVariant, rendererClasses [&NiagaraMeshRendererProperties]}
  {emitterName=FlareVariant, rendererClasses [&NiagaraMeshRendererProperties]}
  {emitterName=FlareVariant_FE, rendererClasses [&NiagaraMeshRendererProperties]}
  {emitterName=BottomSmokeVariant_FE, rendererClasses [&NiagaraMeshRendererProperties]}
  {emitterName=TopSmokeVariant_FE, rendererClasses [&NiagaraMeshRendererProperties]}
 ]
 userVariables [
  User.bIsFrontEnd {type=NiagaraBool}
  User.bIsFrontEndPreview {type=NiagaraBool}
  User.ForwardDot {type=NiagaraFloat, defaultValue=0, description=""}
  User.IsFrontEnd {type=NiagaraBool}
  User.ParaGlideLeanAlpha {type=NiagaraFloat, defaultValue=0, description=""}
  User.Variant {type=NiagaraFloat, defaultValue=1, description=""}
 ]
}
